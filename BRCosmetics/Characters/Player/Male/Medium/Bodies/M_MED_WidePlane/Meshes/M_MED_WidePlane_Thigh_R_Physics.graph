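# Physics asset for the right thigh mesh of the M_MED_WidePlane male player body cosmetic, covering a dynamic pants bone.
# It holds two capsule bodies: a tiny one on calf_r and a larger, offset one on dyn_pants_01_r.
# A single constraint links the pants bone to the calf with narrow swing limits and a tighter twist limit, keeping its movement small.
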PhysicsAsset /BRCosmetics/Characters/Player/Male/Medium/Bodies/M_MED_WidePlane/Meshes/M_MED_WidePlane_Thigh_R_Physics
bodies {
 calf_r [{shapeType=Capsule, length=0.505, radius=0.505, shapeName=calf_r_capsule, rotation=<rot 0 -90 0>}]
 dyn_pants_01_r [{shapeType=Capsule, length=6, radius=6, shapeName=dyn_pants_01_l_capsule, center=<8.539 0 0>, rotation=<rot 90 0 0>}]
}
constraints [{bone1Name=dyn_pants_01_r, bone2Name=calf_r, swing1LimitDegrees=12, swing2LimitDegrees=12, twistLimitDegrees=2}]
