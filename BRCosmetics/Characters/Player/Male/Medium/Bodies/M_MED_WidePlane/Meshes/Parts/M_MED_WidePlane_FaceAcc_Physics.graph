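# Physics asset for the face accessory part of the male medium WidePlane player cosmetic body, covering ten dynamic hair chunk bones on the head.
# Each hair chunk has an identical small capsule body, and the head has a larger capsule.
# Every chunk is constrained directly to the head rather than chained, with narrow swing limits and very tight twist.
# Chunks 06 and 07 get a slightly wider swing range than the others.
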PhysicsAsset /BRCosmetics/Characters/Player/Male/Medium/Bodies/M_MED_WidePlane/Meshes/Parts/M_MED_WidePlane_FaceAcc_Physics
bodies {
 dyn_hair_chunk_01 [{shapeType=Capsule, length=3.5754, radius=2, shapeName=dyn_hair_chunk_01_capsule, center=<2.1949 0 0>, rotation=<rot 90 0 0>}]
 dyn_hair_chunk_02 [{shapeType=Capsule, length=3.5754, radius=2, shapeName=dyn_hair_chunk_02_capsule, center=<2.1949 0 0>, rotation=<rot 90 0 0>}]
 dyn_hair_chunk_03 [{shapeType=Capsule, length=3.5754, radius=2, shapeName=dyn_hair_chunk_03_capsule, center=<2.1949 0 0>, rotation=<rot 90 0 0>}]
 dyn_hair_chunk_04 [{shapeType=Capsule, length=3.5754, radius=2, shapeName=dyn_hair_chunk_04_capsule, center=<2.1949 0 0>, rotation=<rot 90 0 0>}]
 dyn_hair_chunk_05 [{shapeType=Capsule, length=3.5754, radius=2, shapeName=dyn_hair_chunk_05_capsule, center=<2.1949 0 0>, rotation=<rot 90 0 0>}]
 dyn_hair_chunk_06 [{shapeType=Capsule, length=3.5754, radius=2, shapeName=dyn_hair_chunk_06_capsule, center=<2.1949 0 0>, rotation=<rot 90 0 0>}]
 dyn_hair_chunk_07 [{shapeType=Capsule, length=3.5754, radius=2, shapeName=dyn_hair_chunk_07_capsule, center=<2.1949 0 0>, rotation=<rot 90 0 0>}]
 dyn_hair_chunk_08 [{shapeType=Capsule, length=3.5754, radius=2, shapeName=dyn_hair_chunk_08_capsule, center=<2.1949 0 0>, rotation=<rot 90 0 0>}]
 dyn_hair_chunk_09 [{shapeType=Capsule, length=3.5754, radius=2, shapeName=dyn_hair_chunk_09_capsule, center=<2.1949 0 0>, rotation=<rot 90 0 0>}]
 dyn_hair_chunk_10 [{shapeType=Capsule, length=3.5754, radius=2, shapeName=dyn_hair_chunk_10_capsule, center=<2.1949 0 0>, rotation=<rot 90 0 0>}]
 head [{shapeType=Capsule, length=4, radius=5.5112, shapeName=head_capsule, center=<5.4633 0 0>, rotation=<rot 90 0 0>}]
}
constraints [
 {bone1Name=dyn_hair_chunk_01, bone2Name=head, swing1LimitDegrees=16, swing2LimitDegrees=20, twistLimitDegrees=2}
 {bone1Name=dyn_hair_chunk_02, bone2Name=head, swing1LimitDegrees=16, swing2LimitDegrees=20, twistLimitDegrees=2}
 {bone1Name=dyn_hair_chunk_03, bone2Name=head, swing1LimitDegrees=16, swing2LimitDegrees=20, twistLimitDegrees=2}
 {bone1Name=dyn_hair_chunk_04, bone2Name=head, swing1LimitDegrees=16, swing2LimitDegrees=20, twistLimitDegrees=2}
 {bone1Name=dyn_hair_chunk_05, bone2Name=head, swing1LimitDegrees=16, swing2LimitDegrees=20, twistLimitDegrees=2}
 {bone1Name=dyn_hair_chunk_06, bone2Name=head, swing1LimitDegrees=24, swing2LimitDegrees=24, twistLimitDegrees=2}
 {bone1Name=dyn_hair_chunk_07, bone2Name=head, swing1LimitDegrees=24, swing2LimitDegrees=24, twistLimitDegrees=2}
 {bone1Name=dyn_hair_chunk_08, bone2Name=head, swing1LimitDegrees=16, swing2LimitDegrees=20, twistLimitDegrees=2}
 {bone1Name=dyn_hair_chunk_09, bone2Name=head, swing1LimitDegrees=16, swing2LimitDegrees=20, twistLimitDegrees=2}
 {bone1Name=dyn_hair_chunk_10, bone2Name=head, swing1LimitDegrees=16, swing2LimitDegrees=20, twistLimitDegrees=2}
]
